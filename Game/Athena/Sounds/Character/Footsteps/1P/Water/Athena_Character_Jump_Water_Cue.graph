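# Sound cue for the first-person player jump in water, layering soldier water-landing steps with swim splashes.
# Landing samples are randomized, pitched up and shaped by a rising volume envelope, then mixed with random swim samples.
# A quality node gives the first two levels larger sample pools and the last two smaller ones, whose random nodes skip branch culling.
# Concurrency is limited to two per owner, stopping the oldest.
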SoundCue /Game/Athena/Sounds/Character/Footsteps/1P/Water/Athena_Character_Jump_Water_Cue
root=QualityLevel_0
nodes [
 {
  id=Enveloper_0
  children [Modulator_3]
  properties {
   pitchCurve {editorCurveData {keys [{time=0}]}}
   volumeCurve {editorCurveData {keys [<at -0.0002 0.0075>,<at 0.1817 0.0667>,<at 0.3 0.5626>]}}
  }
 }
 {
  id=Enveloper_1
  children [Modulator_1]
  properties {
   pitchCurve {editorCurveData {keys [{time=0}]}}
   volumeCurve {editorCurveData {keys [<at -0.0002 0.0075>,<at 0.1817 0.0667>,<at 0.3 0.5626>]}}
  }
 }
 {id=Mixer_0, children [Modulator_2,Random_3], properties {inputVolume [1,1]}}
 {id=Mixer_1, children [Modulator_0,Random_0], properties {inputVolume [1,1]}}
 {id=Modulator_0, children [Enveloper_1], properties {pitchMax=1.3, pitchMin=1.1, volumeMax=1, volumeMin=0.7}}
 {id=Modulator_1, children [Random_2], properties {pitchMax=1.4, pitchMin=1.2}}
 {id=Modulator_2, children [Enveloper_0], properties {pitchMax=1.3, pitchMin=1.1, volumeMax=1, volumeMin=0.7}}
 {id=Modulator_3, children [Random_1], properties {pitchMax=1.4, pitchMin=1.2}}
 {id=Modulator_4, children [Mixer_0]}
 {id=Modulator_5, children [Mixer_1]}
 {id=QualityLevel_0, children [Modulator_5,Modulator_5,Modulator_4,Modulator_4]}
 {
  id=Random_0
  children [
   WavePlayer_9
   WavePlayer_10
   WavePlayer_11
   WavePlayer_22
   WavePlayer_23
   WavePlayer_24
   WavePlayer_25
   WavePlayer_26
   WavePlayer_27
   WavePlayer_28
  ]
  properties {weights [1,1,1,1,1,1,1,1,1,1]}
 }
 {id=Random_1, children [WavePlayer_0,WavePlayer_7], properties {bShouldExcludeFromBranchCulling=true, weights [1,1]}}
 {id=Random_2, children [WavePlayer_1,WavePlayer_2,WavePlayer_3,WavePlayer_4,WavePlayer_5,WavePlayer_6], properties {weights [1,1,1,1,1,1]}}
 {id=Random_3, children [WavePlayer_12, WavePlayer_13, WavePlayer_14], properties {bShouldExcludeFromBranchCulling=true, weights [1,1,1]}}
 {id=WavePlayer_0, properties {soundWaveAssetPtr=/Game/Sounds/Fort_Step_Sounds/Soldier/Soldier_Step_Land_Water_01}}
 {id=WavePlayer_1, properties {soundWaveAssetPtr=&Soldier_Step_Land_Water_01}}
 {id=WavePlayer_10, properties {soundWaveAssetPtr=/Game/Athena/Sounds/Water/Interaction/Water_Player_Swim_Normal_1P_01}}
 {id=WavePlayer_11, properties {soundWaveAssetPtr=/Game/Athena/Sounds/Water/Interaction/Water_Player_Swim_Normal_1P_02}}
 {id=WavePlayer_12, properties {soundWaveAssetPtr=/Game/Athena/Sounds/Water/Interaction/Water_Player_Swim_Normal_1P_10}}
 {id=WavePlayer_13, properties {soundWaveAssetPtr=&Water_Player_Swim_Normal_1P_01}}
 {id=WavePlayer_14, properties {soundWaveAssetPtr=&Water_Player_Swim_Normal_1P_02}}
 {id=WavePlayer_2, properties {soundWaveAssetPtr=/Game/Sounds/Fort_Step_Sounds/Soldier/Soldier_Step_Land_Water_02}}
 {id=WavePlayer_22, properties {soundWaveAssetPtr=/Game/Athena/Sounds/Water/Interaction/Water_Player_Swim_Normal_1P_03}}
 {id=WavePlayer_23, properties {soundWaveAssetPtr=/Game/Athena/Sounds/Water/Interaction/Water_Player_Swim_Normal_1P_04}}
 {id=WavePlayer_24, properties {soundWaveAssetPtr=/Game/Athena/Sounds/Water/Interaction/Water_Player_Swim_Normal_1P_05}}
 {id=WavePlayer_25, properties {soundWaveAssetPtr=/Game/Athena/Sounds/Water/Interaction/Water_Player_Swim_Normal_1P_06}}
 {id=WavePlayer_26, properties {soundWaveAssetPtr=/Game/Athena/Sounds/Water/Interaction/Water_Player_Swim_Normal_1P_07}}
 {id=WavePlayer_27, properties {soundWaveAssetPtr=/Game/Athena/Sounds/Water/Interaction/Water_Player_Swim_Normal_1P_08}}
 {id=WavePlayer_28, properties {soundWaveAssetPtr=/Game/Athena/Sounds/Water/Interaction/Water_Player_Swim_Normal_1P_09}}
 {id=WavePlayer_3, properties {soundWaveAssetPtr=/Game/Sounds/Fort_Step_Sounds/Soldier/Soldier_Step_Land_Water_03}}
 {id=WavePlayer_4, properties {soundWaveAssetPtr=/Game/Sounds/Fort_Step_Sounds/Soldier/Soldier_Step_Land_Water_04}}
 {id=WavePlayer_5, properties {soundWaveAssetPtr=/Game/Sounds/Fort_Step_Sounds/Soldier/Soldier_Step_Land_Water_05}}
 {id=WavePlayer_6, properties {soundWaveAssetPtr=/Game/Sounds/Fort_Step_Sounds/Soldier/Soldier_Step_Land_Water_06}}
 {id=WavePlayer_7, properties {soundWaveAssetPtr=&Soldier_Step_Land_Water_02}}
 {id=WavePlayer_9, properties {soundWaveAssetPtr=&Water_Player_Swim_Normal_1P_10}}
]
properties {
 attenuationSettings=/Game/Sounds/Attenuation/Fort_Player_Land_Athena_Att
 bOverrideConcurrency=true
 maxDistance=2097152
 soundClassObject=&Fort_Player_Foley_1P_Athena
 volumeMultiplier=1.25
 concurrencyOverrides {bLimitToOwner=true, maxCount=2, resolutionRule=StopOldest}
}
